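# NPC conversation database for the assault weapon expertise Expert Sale service, selling loot in five tiers from C to SR.
# The entry fans out to five EXPERT SALE speech choices gated on that service, each offering BUY or BACK.
# Each sale checks the player's assault expertise token count against curve rows (exact match above the lowest tier), hiding on failure.
# Costs are curve-driven, and all purchases share an apply-effect node that returns to start.
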ConversationDatabase /NPCLibrary/ConversationsShared/ExpertiseConversations/Conversation_Expertise_Assault
entry_points [{tag=Conversation.Entry.Expertise.Assault, destinations [#1]}]
nodes [
 ConversationEntryPointNode_3 {
  id=#1
  class=ConversationEntryPointNode
  outputs [
   #2
   #3
   #4
   #5
   #6
  ]
  properties {entryTag {tagName=Conversation.Entry.Expertise.Assault}}
 }
 Speech_C_4 {
  id=#6
  class=Speech_C
  outputs [#7, #8]
  properties {
   bUsePerNPC_ContextualMessageLookup=true
   softTaskIcon=&T_UI_NPC_WeaponExpertise
   generalConfig {defaultMessage=You want this?}
   speakerEntryTagToConfig {"(TagName=\"Conversation.Entry.TealMink\")" {defaultMessage=You want this?}}
  }
  sub_nodes [
   FortConversationChoiceWithPreviewNode_6 {class=FortConversationChoiceWithPreviewNode, properties {defaultChoiceDisplayText=EXPERT SALE}}
   FortConversationRequirement_HasService_7 {class=FortConversationRequirement_HasService, properties {serviceTag {tagName=Tandem.Service.ExpertSale}}}
  ]
 }
 ExpertSale_C_9 {
  id=#8
  class=ExpertSale_C
  outputs [#9]
  properties {
   bShowPropertyDetails=false
   loot Tier Group Name=LTG_NPC_Expertise_Assault_CQB_AR_SR
   requirements [{
     cannotUseReasonParameter=""
     failureNodeBehaviour=FailedAndHidden
     participantId {tagName=Conversation.Participant.Player}
     requirement {
      class=UFortControllerRequirement_TokenThreshold_C
      properties {
       itemTemplateId=Token:athena_s34_weaponexpertise_assault_service_token
       requiredQuantityThresholdConfig {
        requirement=Equal
        quantity {
         value=1
         curve {curveTable=/Game/Conversation/Services/ExpertSale/DataTables/ExpertSaleRequirementsData, rowName=ExpertSale.Requirement.Tokens.5}
        }
       }
      }
     }
    }]
   serviceCost {curve {rowName=ExpertSale.Cost.Assault.Five}}
  }
  sub_nodes [ConversationChoiceNode_10 {class=ConversationChoiceNode, properties {defaultChoiceDisplayText=BUY}}]
 }
 FortConversationTaskNode_ApplyDataDrivenEffect_2 {id=#9, class=FortConversationTaskNode_ApplyDataDrivenEffect, outputs [#10]}
 ReturnToStart_C_0 {id=#10, class=ReturnToStart_C}
 Back_C_3 {
  id=#7
  class=Back_C
  sub_nodes [ConversationChoiceNode_6 {class=ConversationChoiceNode, properties {defaultChoiceDisplayText=BACK}}]
 }
 Speech_C_2 {
  id=#5
  class=Speech_C
  outputs [#11, #12]
  properties {
   bUsePerNPC_ContextualMessageLookup=true
   softTaskIcon=&T_UI_NPC_WeaponExpertise
   generalConfig {defaultMessage=You want this?}
   speakerEntryTagToConfig {"(TagName=\"Conversation.Entry.TealMink\")" {defaultMessage=You want this?}}
  }
  sub_nodes [
   FortConversationChoiceWithPreviewNode_3 {class=FortConversationChoiceWithPreviewNode, properties {defaultChoiceDisplayText=EXPERT SALE}}
   FortConversationRequirement_HasService_3 {class=FortConversationRequirement_HasService, properties {serviceTag {tagName=Tandem.Service.ExpertSale}}}
  ]
 }
 ExpertSale_C_8 {
  id=#12
  class=ExpertSale_C
  outputs [#9]
  properties {
   bShowPropertyDetails=false
   loot Tier Group Name=LTG_NPC_Expertise_Assault_CQB_AR_VR
   requirements [{
     cannotUseReasonParameter=""
     failureNodeBehaviour=FailedAndHidden
     participantId {tagName=Conversation.Participant.Player}
     requirement {
      class=UFortControllerRequirement_TokenThreshold_C
      properties {
       itemTemplateId=Token:athena_s34_weaponexpertise_assault_service_token
       requiredQuantityThresholdConfig {
        requirement=Equal
        quantity {
         value=1
         curve {curveTable=&ExpertSaleRequirementsData, rowName=ExpertSale.Requirement.Tokens.4}
        }
       }
      }
     }
    }]
   serviceCost {curve {rowName=ExpertSale.Cost.Assault.Four}}
  }
  sub_nodes [ConversationChoiceNode_14 {class=ConversationChoiceNode, properties {defaultChoiceDisplayText=BUY}}]
 }
 Back_C_2 {
  id=#11
  class=Back_C
  sub_nodes [ConversationChoiceNode_9 {class=ConversationChoiceNode, properties {defaultChoiceDisplayText=BACK}}]
 }
 Speech_C_1 {
  id=#4
  class=Speech_C
  outputs [#13, #14]
  properties {
   bUsePerNPC_ContextualMessageLookup=true
   softTaskIcon=&T_UI_NPC_WeaponExpertise
   generalConfig {defaultMessage=You want this?}
   speakerEntryTagToConfig {"(TagName=\"Conversation.Entry.TealMink\")" {defaultMessage=You want this?}}
  }
  sub_nodes [
   FortConversationChoiceWithPreviewNode_2 {class=FortConversationChoiceWithPreviewNode, properties {defaultChoiceDisplayText=EXPERT SALE}}
   FortConversationRequirement_HasService_2 {class=FortConversationRequirement_HasService, properties {serviceTag {tagName=Tandem.Service.ExpertSale}}}
  ]
 }
 ExpertSale_C_7 {
  id=#14
  class=ExpertSale_C
  outputs [#9]
  properties {
   bShowPropertyDetails=false
   loot Tier Group Name=LTG_NPC_Expertise_Assault_CQB_AR_R
   requirements [{
     cannotUseReasonParameter=""
     failureNodeBehaviour=FailedAndHidden
     participantId {tagName=Conversation.Participant.Player}
     requirement {
      class=UFortControllerRequirement_TokenThreshold_C
      properties {
       itemTemplateId=Token:athena_s34_weaponexpertise_assault_service_token
       requiredQuantityThresholdConfig {
        requirement=Equal
        quantity {
         value=1
         curve {curveTable=&ExpertSaleRequirementsData, rowName=ExpertSale.Requirement.Tokens.3}
        }
       }
      }
     }
    }]
   serviceCost {curve {rowName=ExpertSale.Cost.Assault.Three}}
  }
  sub_nodes [ConversationChoiceNode_13 {class=ConversationChoiceNode, properties {defaultChoiceDisplayText=BUY}}]
 }
 Back_C_1 {
  id=#13
  class=Back_C
  sub_nodes [ConversationChoiceNode_5 {class=ConversationChoiceNode, properties {defaultChoiceDisplayText=BACK}}]
 }
 Speech_C_0 {
  id=#3
  class=Speech_C
  outputs [#15, #16]
  properties {
   bUsePerNPC_ContextualMessageLookup=true
   softTaskIcon=&T_UI_NPC_WeaponExpertise
   generalConfig {defaultMessage=You want this?}
   speakerEntryTagToConfig {"(TagName=\"Conversation.Entry.TealMink\")" {defaultMessage=You want this?}}
  }
  sub_nodes [
   FortConversationChoiceWithPreviewNode_1 {class=FortConversationChoiceWithPreviewNode, properties {defaultChoiceDisplayText=EXPERT SALE}}
   FortConversationRequirement_HasService_1 {class=FortConversationRequirement_HasService, properties {serviceTag {tagName=Tandem.Service.ExpertSale}}}
  ]
 }
 ExpertSale_C_6 {
  id=#16
  class=ExpertSale_C
  outputs [#9]
  properties {
   bShowPropertyDetails=false
   loot Tier Group Name=LTG_NPC_Expertise_Assault_CQB_AR_UC
   requirements [{
     cannotUseReasonParameter=""
     failureNodeBehaviour=FailedAndHidden
     participantId {tagName=Conversation.Participant.Player}
     requirement {
      class=UFortControllerRequirement_TokenThreshold_C
      properties {
       itemTemplateId=Token:athena_s34_weaponexpertise_assault_service_token
       requiredQuantityThresholdConfig {
        requirement=Equal
        quantity {
         value=1
         curve {curveTable=&ExpertSaleRequirementsData, rowName=ExpertSale.Requirement.Tokens.2}
        }
       }
      }
     }
    }]
   serviceCost {curve {rowName=ExpertSale.Cost.Assault.Two}}
  }
  sub_nodes [ConversationChoiceNode_12 {class=ConversationChoiceNode, properties {defaultChoiceDisplayText=BUY}}]
 }
 Back_C_0 {
  id=#15
  class=Back_C
  sub_nodes [ConversationChoiceNode_3 {class=ConversationChoiceNode, properties {defaultChoiceDisplayText=BACK}}]
 }
 Speech_C_3 {
  id=#2
  class=Speech_C
  outputs [#17, #18]
  properties {
   bUsePerNPC_ContextualMessageLookup=true
   softTaskIcon=&T_UI_NPC_WeaponExpertise
   generalConfig {defaultMessage=You want this?}
   speakerEntryTagToConfig {"(TagName=\"Conversation.Entry.TealMink\")" {defaultMessage=You want this?}}
  }
  sub_nodes [
   FortConversationChoiceWithPreviewNode_0 {class=FortConversationChoiceWithPreviewNode, properties {defaultChoiceDisplayText=EXPERT SALE}}
   FortConversationRequirement_HasService_0 {class=FortConversationRequirement_HasService, properties {serviceTag {tagName=Tandem.Service.ExpertSale}}}
  ]
 }
 ExpertSale_C_5 {
  id=#18
  class=ExpertSale_C
  outputs [#9]
  properties {
   bHideWhenFailed=true
   bShowPropertyDetails=false
   loot Tier Group Name=LTG_NPC_Expertise_Assault_CQB_AR_C
   requirements [{
     cannotUseReasonParameter=""
     failureNodeBehaviour=FailedAndHidden
     participantId {tagName=Conversation.Participant.Player}
     requirement {
      class=UFortControllerRequirement_TokenThreshold_C
      properties {
       itemTemplateId=Token:athena_s34_weaponexpertise_assault_service_token
       requiredQuantityThresholdConfig {
        requirement=LessThanOrEqual
        quantity {
         value=1
         curve {curveTable=&ExpertSaleRequirementsData, rowName=ExpertSale.Requirement.Tokens.1}
        }
       }
      }
     }
    }]
   serviceCost {curve {rowName=ExpertSale.Cost.Assault.One}}
  }
  sub_nodes [ConversationChoiceNode_11 {class=ConversationChoiceNode, properties {defaultChoiceDisplayText=BUY}}]
 }
 Back_C_7 {
  id=#17
  class=Back_C
  sub_nodes [ConversationChoiceNode_1 {class=ConversationChoiceNode, properties {defaultChoiceDisplayText=BACK}}]
 }
]
