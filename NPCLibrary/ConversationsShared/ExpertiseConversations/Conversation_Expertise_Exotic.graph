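# Conversation database for an NPC "Expert Sale" service that sells exotic weapon expertise loot to the player.
# The entry point offers five parallel speech choices, gated on the NPC having the ExpertSale service, each leading to Back or a single-sale purchase.
# Each sale hides unless the player's exotic expertise token count matches its tier threshold, selecting that tier's curve-driven price from one shared loot group.
# All purchases converge on a data-driven effect node, then return to the conversation start.
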ConversationDatabase /NPCLibrary/ConversationsShared/ExpertiseConversations/Conversation_Expertise_Exotic
entry_points [{tag=Conversation.Entry.Expertise.Exotic, destinations [#1]}]
nodes [
 ConversationEntryPointNode_3 {
  id=#1
  class=ConversationEntryPointNode
  outputs [
   #2
   #3
   #4
   #5
   #6
  ]
  properties {entryTag {tagName=Conversation.Entry.Expertise.Exotic}}
 }
 Speech_C_7 {
  id=#6
  class=Speech_C
  outputs [#7, #8]
  properties {
   bUsePerNPC_ContextualMessageLookup=true
   softTaskIcon=&T_UI_NPC_WeaponExpertise
   generalConfig {defaultMessage=For experts only.}
   speakerEntryTagToConfig {"(TagName=\"Conversation.Entry.AuraCop\")" {defaultMessage=A refined choice.}}
  }
  sub_nodes [
   FortConversationChoiceWithPreviewNode_4 {class=FortConversationChoiceWithPreviewNode, properties {defaultChoiceDisplayText=EXPERT SALE}}
   FortConversationRequirement_HasService_4 {class=FortConversationRequirement_HasService, properties {serviceTag {tagName=Tandem.Service.ExpertSale}}}
  ]
 }
 ExpertSale_Single_C_4 {
  id=#8
  class=ExpertSale_Single_C
  outputs [#9]
  properties {
   bShowPropertyDetails=false
   loot Tier Group Name=LTG_NPC_Expertise_Exotic_Diamond
   requirements [{
     cannotUseReasonParameter=""
     failureNodeBehaviour=FailedAndHidden
     participantId {tagName=Conversation.Participant.Player}
     requirement {
      class=UFortControllerRequirement_TokenThreshold_C
      properties {
       itemTemplateId=Token:athena_s34_weaponexpertise_exotic_service_token
       requiredQuantityThresholdConfig {
        requirement=Equal
        quantity {
         value=1
         curve {curveTable=/Game/Conversation/Services/ExpertSale/DataTables/ExpertSaleRequirementsData, rowName=ExpertSale.Requirement.Tokens.5}
        }
       }
      }
     }
    }]
   serviceCost {curve {rowName=ExpertSale.Cost.Exotic.Five}}
  }
  sub_nodes [ConversationChoiceNode_2 {class=ConversationChoiceNode, properties {defaultChoiceDisplayText=BUY}}]
 }
 FortConversationTaskNode_ApplyDataDrivenEffect_2 {id=#9, class=FortConversationTaskNode_ApplyDataDrivenEffect, outputs [#10]}
 ReturnToStart_C_0 {id=#10, class=ReturnToStart_C}
 Back_C_5 {
  id=#7
  class=Back_C
  sub_nodes [ConversationChoiceNode_0 {class=ConversationChoiceNode, properties {defaultChoiceDisplayText=BACK}}]
 }
 Speech_C_6 {
  id=#5
  class=Speech_C
  outputs [#11, #12]
  properties {
   bUsePerNPC_ContextualMessageLookup=true
   softTaskIcon=&T_UI_NPC_WeaponExpertise
   generalConfig {defaultMessage=For experts only.}
   speakerEntryTagToConfig {"(TagName=\"Conversation.Entry.AuraCop\")" {defaultMessage=A refined choice.}}
  }
  sub_nodes [
   FortConversationChoiceWithPreviewNode_7 {class=FortConversationChoiceWithPreviewNode, properties {defaultChoiceDisplayText=EXPERT SALE}}
   FortConversationRequirement_HasService_8 {class=FortConversationRequirement_HasService, properties {serviceTag {tagName=Tandem.Service.ExpertSale}}}
  ]
 }
 ExpertSale_Single_C_3 {
  id=#12
  class=ExpertSale_Single_C
  outputs [#9]
  properties {
   bShowPropertyDetails=false
   loot Tier Group Name=LTG_NPC_Expertise_Exotic_Diamond
   requirements [{
     cannotUseReasonParameter=""
     failureNodeBehaviour=FailedAndHidden
     participantId {tagName=Conversation.Participant.Player}
     requirement {
      class=UFortControllerRequirement_TokenThreshold_C
      properties {
       itemTemplateId=Token:athena_s34_weaponexpertise_exotic_service_token
       requiredQuantityThresholdConfig {
        requirement=Equal
        quantity {
         value=1
         curve {curveTable=&ExpertSaleRequirementsData, rowName=ExpertSale.Requirement.Tokens.4}
        }
       }
      }
     }
    }]
   serviceCost {curve {rowName=ExpertSale.Cost.Exotic.Four}}
  }
  sub_nodes [ConversationChoiceNode_5 {class=ConversationChoiceNode, properties {defaultChoiceDisplayText=BUY}}]
 }
 Back_C_4 {
  id=#11
  class=Back_C
  sub_nodes [ConversationChoiceNode_19 {class=ConversationChoiceNode, properties {defaultChoiceDisplayText=BACK}}]
 }
 Speech_C_4 {
  id=#4
  class=Speech_C
  outputs [#13, #14]
  properties {
   bUsePerNPC_ContextualMessageLookup=true
   softTaskIcon=&T_UI_NPC_WeaponExpertise
   generalConfig {defaultMessage=For experts only.}
   speakerEntryTagToConfig {"(TagName=\"Conversation.Entry.AuraCop\")" {defaultMessage=A refined choice.}}
  }
  sub_nodes [
   FortConversationChoiceWithPreviewNode_6 {class=FortConversationChoiceWithPreviewNode, properties {defaultChoiceDisplayText=EXPERT SALE}}
   FortConversationRequirement_HasService_7 {class=FortConversationRequirement_HasService, properties {serviceTag {tagName=Tandem.Service.ExpertSale}}}
  ]
 }
 ExpertSale_Single_C_2 {
  id=#14
  class=ExpertSale_Single_C
  outputs [#9]
  properties {
   bShowPropertyDetails=false
   loot Tier Group Name=LTG_NPC_Expertise_Exotic_Diamond
   requirements [{
     cannotUseReasonParameter=""
     failureNodeBehaviour=FailedAndHidden
     participantId {tagName=Conversation.Participant.Player}
     requirement {
      class=UFortControllerRequirement_TokenThreshold_C
      properties {
       itemTemplateId=Token:athena_s34_weaponexpertise_exotic_service_token
       requiredQuantityThresholdConfig {
        requirement=Equal
        quantity {
         value=1
         curve {curveTable=&ExpertSaleRequirementsData, rowName=ExpertSale.Requirement.Tokens.3}
        }
       }
      }
     }
    }]
   serviceCost {curve {rowName=ExpertSale.Cost.Exotic.Three}}
  }
  sub_nodes [ConversationChoiceNode_4 {class=ConversationChoiceNode, properties {defaultChoiceDisplayText=BUY}}]
 }
 Back_C_3 {
  id=#13
  class=Back_C
  sub_nodes [ConversationChoiceNode_17 {class=ConversationChoiceNode, properties {defaultChoiceDisplayText=BACK}}]
 }
 Speech_C_5 {
  id=#3
  class=Speech_C
  outputs [#15, #16]
  properties {
   bUsePerNPC_ContextualMessageLookup=true
   softTaskIcon=&T_UI_NPC_WeaponExpertise
   generalConfig {defaultMessage=For experts only.}
   speakerEntryTagToConfig {"(TagName=\"Conversation.Entry.AuraCop\")" {defaultMessage=A refined choice.}}
  }
  sub_nodes [
   FortConversationChoiceWithPreviewNode_5 {class=FortConversationChoiceWithPreviewNode, properties {defaultChoiceDisplayText=EXPERT SALE}}
   FortConversationRequirement_HasService_6 {class=FortConversationRequirement_HasService, properties {serviceTag {tagName=Tandem.Service.ExpertSale}}}
  ]
 }
 ExpertSale_Single_C_1 {
  id=#16
  class=ExpertSale_Single_C
  outputs [#9]
  properties {
   bShowPropertyDetails=false
   loot Tier Group Name=LTG_NPC_Expertise_Exotic_Diamond
   requirements [{
     cannotUseReasonParameter=""
     failureNodeBehaviour=FailedAndHidden
     participantId {tagName=Conversation.Participant.Player}
     requirement {
      class=UFortControllerRequirement_TokenThreshold_C
      properties {
       itemTemplateId=Token:athena_s34_weaponexpertise_exotic_service_token
       requiredQuantityThresholdConfig {
        requirement=Equal
        quantity {
         value=1
         curve {curveTable=&ExpertSaleRequirementsData, rowName=ExpertSale.Requirement.Tokens.2}
        }
       }
      }
     }
    }]
   serviceCost {curve {rowName=ExpertSale.Cost.Exotic.Two}}
  }
  sub_nodes [ConversationChoiceNode_3 {class=ConversationChoiceNode, properties {defaultChoiceDisplayText=BUY}}]
 }
 Back_C_2 {
  id=#15
  class=Back_C
  sub_nodes [ConversationChoiceNode_9 {class=ConversationChoiceNode, properties {defaultChoiceDisplayText=BACK}}]
 }
 Speech_C_2 {
  id=#2
  class=Speech_C
  outputs [#17, #18]
  properties {
   bUsePerNPC_ContextualMessageLookup=true
   softTaskIcon=&T_UI_NPC_WeaponExpertise
   generalConfig {defaultMessage=For experts only.}
   speakerEntryTagToConfig {"(TagName=\"Conversation.Entry.AuraCop\")" {defaultMessage=A refined choice.}}
  }
  sub_nodes [
   FortConversationChoiceWithPreviewNode_0 {class=FortConversationChoiceWithPreviewNode, properties {defaultChoiceDisplayText=EXPERT SALE}}
   FortConversationRequirement_HasService_1 {class=FortConversationRequirement_HasService, properties {serviceTag {tagName=Tandem.Service.ExpertSale}}}
  ]
 }
 ExpertSale_Single_C_0 {
  id=#18
  class=ExpertSale_Single_C
  outputs [#9]
  properties {
   bHideWhenFailed=true
   bShowPropertyDetails=false
   loot Tier Group Name=LTG_NPC_Expertise_Exotic_Diamond
   requirements [{
     cannotUseReasonParameter=""
     failureNodeBehaviour=FailedAndHidden
     participantId {tagName=Conversation.Participant.Player}
     requirement {
      class=UFortControllerRequirement_TokenThreshold_C
      properties {
       itemTemplateId=Token:athena_s34_weaponexpertise_exotic_service_token
       requiredQuantityThresholdConfig {
        requirement=LessThanOrEqual
        quantity {
         value=1
         curve {curveTable=&ExpertSaleRequirementsData, rowName=ExpertSale.Requirement.Tokens.1}
        }
       }
      }
     }
    }]
   serviceCost {curve {rowName=ExpertSale.Cost.Exotic.One}}
  }
  sub_nodes [ConversationChoiceNode_1 {class=ConversationChoiceNode, properties {defaultChoiceDisplayText=BUY}}]
 }
 Back_C_1 {
  id=#17
  class=Back_C
  sub_nodes [ConversationChoiceNode_7 {class=ConversationChoiceNode, properties {defaultChoiceDisplayText=BACK}}]
 }
]
