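# NPC conversation database that sells toggle-zoom sniper rifles as an Expert Sale tied to the player's sniper expertise.
# The entry fans out to five parallel speech choices, each requiring the ExpertSale service and offering Back or Buy.
# Each sale checks a curve-driven expertise token count, hiding on failure, and sets its loot tier and cost, with rarity rising across tiers.
# All purchases share one data-driven effect task, then return to start.
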ConversationDatabase /NPCLibrary/ConversationsShared/ExpertiseConversations/Conversation_Expertise_Sniper
entry_points [{tag=Conversation.Entry.Expertise.Sniper, destinations [#1]}]
nodes [
 ConversationEntryPointNode_0 {
  id=#1
  class=ConversationEntryPointNode
  outputs [
   #2
   #3
   #4
   #5
   #6
  ]
  properties {entryTag {tagName=Conversation.Entry.Expertise.Sniper}}
 }
 Speech_C_9 {
  id=#6
  class=Speech_C
  outputs [#7, #8]
  properties {
   bUsePerNPC_ContextualMessageLookup=true
   softTaskIcon=&T_UI_NPC_WeaponExpertise
   generalConfig {defaultMessage=Take 'em out.}
   speakerEntryTagToConfig {"(TagName=\"Conversation.Entry.CasinoReaper\")" {defaultMessage=Take 'em out.}}
  }
  sub_nodes [
   FortConversationChoiceWithPreviewNode_1 {class=FortConversationChoiceWithPreviewNode, properties {defaultChoiceDisplayText=EXPERT SALE}}
   FortConversationRequirement_HasService_2 {class=FortConversationRequirement_HasService, properties {serviceTag {tagName=Tandem.Service.ExpertSale}}}
  ]
 }
 ExpertSale_C_10 {
  id=#8
  class=ExpertSale_C
  outputs [#9]
  properties {
   bShowPropertyDetails=false
   loot Tier Group Name=LTG_NPC_Expertise_Sniper_ToggleZoom_SR
   extraData [
    item {value=/VividRazorWeaponGameplay/Gameplay/SniperToggleZoom/WID_ToggleZoomSniper_SR}
    cost {value=360}
    remaining {value=1}
    stacksize {value=1}
    Type {value=service.expertsale}
    preview_tags {value=(GameplayTags=((TagName="Conversation.Choice.Category.Service")))}
    BriefWidgetClassPath {value=/FortniteConversationUI/Conversation/WBP_FortBuyBrief_ExpertSale.WBP_FortBuyBrief_ExpertSale_C}
    stockcount {value=1}
    configdef {value=/Game/Conversation/Services/ExpertSale/ExpertSaleBriefCollection}
    configkey {value=ExpertSale}
   ]
   requirements [{
     cannotUseReasonParameter=""
     failureNodeBehaviour=FailedAndHidden
     participantId {tagName=Conversation.Participant.Player}
     requirement {
      class=UFortControllerRequirement_TokenThreshold_C
      properties {
       itemTemplateId=Token:athena_s34_weaponexpertise_sniper_service_token
       requiredQuantityThresholdConfig {
        requirement=Equal
        quantity {
         value=1
         curve {curveTable=/Game/Conversation/Services/ExpertSale/DataTables/ExpertSaleRequirementsData, rowName=ExpertSale.Requirement.Tokens.5}
        }
       }
      }
     }
    }]
   serviceCost {curve {rowName=ExpertSale.Cost.Sniper.Five}}
  }
  sub_nodes [ConversationChoiceNode_4 {class=ConversationChoiceNode, properties {defaultChoiceDisplayText=BUY}}]
 }
 FortConversationTaskNode_ApplyDataDrivenEffect_2 {id=#9, class=FortConversationTaskNode_ApplyDataDrivenEffect, outputs [#10]}
 ReturnToStart_C_1 {id=#10, class=ReturnToStart_C}
 Back_C_10 {
  id=#7
  class=Back_C
  sub_nodes [ConversationChoiceNode_10 {class=ConversationChoiceNode, properties {defaultChoiceDisplayText=BACK}}]
 }
 Speech_C_8 {
  id=#5
  class=Speech_C
  outputs [#11, #12]
  properties {
   bUsePerNPC_ContextualMessageLookup=true
   softTaskIcon=&T_UI_NPC_WeaponExpertise
   generalConfig {defaultMessage=Take 'em out.}
   speakerEntryTagToConfig {"(TagName=\"Conversation.Entry.CasinoReaper\")" {defaultMessage=Take 'em out.}}
  }
  sub_nodes [
   FortConversationChoiceWithPreviewNode_9 {class=FortConversationChoiceWithPreviewNode, properties {defaultChoiceDisplayText=EXPERT SALE}}
   FortConversationRequirement_HasService_9 {class=FortConversationRequirement_HasService, properties {serviceTag {tagName=Tandem.Service.ExpertSale}}}
  ]
 }
 ExpertSale_C_6 {
  id=#12
  class=ExpertSale_C
  outputs [#9]
  properties {
   bShowPropertyDetails=false
   loot Tier Group Name=LTG_NPC_Expertise_Sniper_ToggleZoom_VR
   requirements [{
     cannotUseReasonParameter=""
     failureNodeBehaviour=FailedAndHidden
     participantId {tagName=Conversation.Participant.Player}
     requirement {
      class=UFortControllerRequirement_TokenThreshold_C
      properties {
       itemTemplateId=Token:athena_s34_weaponexpertise_sniper_service_token
       requiredQuantityThresholdConfig {
        requirement=Equal
        quantity {
         value=1
         curve {curveTable=&ExpertSaleRequirementsData, rowName=ExpertSale.Requirement.Tokens.4}
        }
       }
      }
     }
    }]
   serviceCost {curve {rowName=ExpertSale.Cost.Sniper.Four}}
  }
  sub_nodes [ConversationChoiceNode_7 {class=ConversationChoiceNode, properties {defaultChoiceDisplayText=BUY}}]
 }
 Back_C_9 {
  id=#11
  class=Back_C
  sub_nodes [ConversationChoiceNode_23 {class=ConversationChoiceNode, properties {defaultChoiceDisplayText=BACK}}]
 }
 Speech_C_0 {
  id=#4
  class=Speech_C
  outputs [#13, #14]
  properties {
   bUsePerNPC_ContextualMessageLookup=true
   softTaskIcon=&T_UI_NPC_WeaponExpertise
   generalConfig {defaultMessage=Take 'em out.}
   speakerEntryTagToConfig {"(TagName=\"Conversation.Entry.CasinoReaper\")" {defaultMessage=Take 'em out.}}
  }
  sub_nodes [
   FortConversationChoiceWithPreviewNode_8 {class=FortConversationChoiceWithPreviewNode, properties {defaultChoiceDisplayText=EXPERT SALE}}
   FortConversationRequirement_HasService_5 {class=FortConversationRequirement_HasService, properties {serviceTag {tagName=Tandem.Service.ExpertSale}}}
  ]
 }
 ExpertSale_C_5 {
  id=#14
  class=ExpertSale_C
  outputs [#9]
  properties {
   bShowPropertyDetails=false
   loot Tier Group Name=LTG_NPC_Expertise_Sniper_ToggleZoom_VR
   requirements [{
     cannotUseReasonParameter=""
     failureNodeBehaviour=FailedAndHidden
     participantId {tagName=Conversation.Participant.Player}
     requirement {
      class=UFortControllerRequirement_TokenThreshold_C
      properties {
       itemTemplateId=Token:athena_s34_weaponexpertise_sniper_service_token
       requiredQuantityThresholdConfig {
        requirement=Equal
        quantity {
         value=1
         curve {curveTable=&ExpertSaleRequirementsData, rowName=ExpertSale.Requirement.Tokens.3}
        }
       }
      }
     }
    }]
   serviceCost {curve {rowName=Expertsale.Cost.Sniper.Three}}
  }
  sub_nodes [ConversationChoiceNode_6 {class=ConversationChoiceNode, properties {defaultChoiceDisplayText=BUY}}]
 }
 Back_C_8 {
  id=#13
  class=Back_C
  sub_nodes [ConversationChoiceNode_21 {class=ConversationChoiceNode, properties {defaultChoiceDisplayText=BACK}}]
 }
 Speech_C_1 {
  id=#3
  class=Speech_C
  outputs [#15, #16]
  properties {
   bUsePerNPC_ContextualMessageLookup=true
   softTaskIcon=&T_UI_NPC_WeaponExpertise
   generalConfig {defaultMessage=Take 'em out.}
   speakerEntryTagToConfig {"(TagName=\"Conversation.Entry.CasinoReaper\")" {defaultMessage=Take 'em out.}}
  }
  sub_nodes [
   FortConversationChoiceWithPreviewNode_3 {class=FortConversationChoiceWithPreviewNode, properties {defaultChoiceDisplayText=EXPERT SALE}}
   FortConversationRequirement_HasService_3 {class=FortConversationRequirement_HasService, properties {serviceTag {tagName=Tandem.Service.ExpertSale}}}
  ]
 }
 ExpertSale_C_2 {
  id=#16
  class=ExpertSale_C
  outputs [#9]
  properties {
   bShowPropertyDetails=false
   loot Tier Group Name=LTG_NPC_Expertise_Sniper_ToggleZoom_R
   requirements [{
     cannotUseReasonParameter=""
     failureNodeBehaviour=FailedAndHidden
     participantId {tagName=Conversation.Participant.Player}
     requirement {
      class=UFortControllerRequirement_TokenThreshold_C
      properties {
       itemTemplateId=Token:athena_s34_weaponexpertise_sniper_service_token
       requiredQuantityThresholdConfig {
        requirement=Equal
        quantity {
         value=1
         curve {curveTable=&ExpertSaleRequirementsData, rowName=ExpertSale.Requirement.Tokens.2}
        }
       }
      }
     }
    }]
   serviceCost {curve {rowName=ExpertSale.Cost.Sniper.Two}}
  }
  sub_nodes [ConversationChoiceNode_5 {class=ConversationChoiceNode, properties {defaultChoiceDisplayText=BUY}}]
 }
 Back_C_6 {
  id=#15
  class=Back_C
  sub_nodes [ConversationChoiceNode_12 {class=ConversationChoiceNode, properties {defaultChoiceDisplayText=BACK}}]
 }
 Speech_C_3 {
  id=#2
  class=Speech_C
  outputs [#17, #18]
  properties {
   bUsePerNPC_ContextualMessageLookup=true
   softTaskIcon=&T_UI_NPC_WeaponExpertise
   generalConfig {defaultMessage=Take 'em out.}
   speakerEntryTagToConfig {"(TagName=\"Conversation.Entry.CasinoReaper\")" {defaultMessage=Take 'em out.}}
  }
  sub_nodes [
   FortConversationChoiceWithPreviewNode_4 {class=FortConversationChoiceWithPreviewNode, properties {defaultChoiceDisplayText=EXPERT SALE}}
   FortConversationRequirement_HasService_4 {class=FortConversationRequirement_HasService, properties {serviceTag {tagName=Tandem.Service.ExpertSale}}}
  ]
 }
 ExpertSale_C_1 {
  id=#18
  class=ExpertSale_C
  outputs [#9]
  properties {
   bHideWhenFailed=true
   bShowPropertyDetails=false
   loot Tier Group Name=LTG_NPC_Expertise_Sniper_ToggleZoom_R
   extraData [
    item {value=/VividRazorWeaponGameplay/Gameplay/SniperToggleZoom/WID_ToggleZoomSniper_R}
    cost {value=300}
    remaining {value=1}
    stacksize {value=1}
    Type {value=service.expertsale}
    preview_tags {value=(GameplayTags=((TagName="Conversation.Choice.Category.Service")))}
    BriefWidgetClassPath {value=&WBP_FortBuyBrief_ExpertSale_C}
    locked {value=1}
    unlock_level {value=-1}
    locked_text {value="LOCTEXT(\"Requires Sniper Expertise\")"}
    stockcount {value=1}
    configdef {value=&ExpertSaleBriefCollection}
    configkey {value=ExpertSale}
   ]
   requirements [{
     cannotUseReasonParameter=""
     failureNodeBehaviour=FailedAndHidden
     participantId {tagName=Conversation.Participant.Player}
     requirement {
      class=UFortControllerRequirement_TokenThreshold_C
      properties {
       itemTemplateId=Token:athena_s34_weaponexpertise_sniper_service_token
       requiredQuantityThresholdConfig {
        requirement=LessThanOrEqual
        quantity {
         value=1
         curve {curveTable=&ExpertSaleRequirementsData, rowName=ExpertSale.Requirement.Tokens.1}
        }
       }
      }
     }
    }]
   serviceCost {curve {rowName=ExpertSale.Cost.Sniper.One}}
  }
  sub_nodes [ConversationChoiceNode_2 {class=ConversationChoiceNode, properties {defaultChoiceDisplayText=BUY}}]
 }
 Back_C_7 {
  id=#17
  class=Back_C
  sub_nodes [ConversationChoiceNode_1 {class=ConversationChoiceNode, properties {defaultChoiceDisplayText=BACK}}]
 }
]
